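# Physics asset for the Llama Hero Winter backpack cosmetic, defining collision bodies and joints for its dangling accessories.
# Sphere bodies on base bones root short capsule or box chains for a carabiner-hung bottle, zipper, pack loop, left and right ropes, and underpack rope and loop.
# Constraints use narrow swing and twist limits, keeping motion subtle.
# The pack offset bone carries an unconstrained capsule-and-box compound body approximating the pack.
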PhysicsAsset /BRCosmetics/Accessories/FORT_Backpacks/Backpack_M_MED_Llama_Hero_Winter/Meshes/M_MED_Llama_Hero_Winter_Pack_Physics
bodies {
 dyn_bottle [{shapeType=Capsule, length=12.9439, radius=3.1379, center=<9.1975 -0.5415 0.7332>, rotation=<rot 86.5736 171.4768 -179.9999>}]
 dyn_bottle_carabiner_base [{shapeType=Sphere, radius=0.5}]
 dyn_bottle_carabiner_parent [{shapeType=Capsule, length=2.2481, radius=1.3405, center=<2.2241 0 0>, rotation=<rot 90 0 0>}]
 dyn_pack_loop_base [{shapeType=Sphere, radius=0.5}]
 dyn_pack_loop_parent [{shapeType=Box, extentX=3.2868, extentY=4.6533, extentZ=1.3578, center=<1.5234 0 0>}]
 dyn_pack_offset [
  {shapeType=Capsule, length=12.75, radius=6.375, center=<9.1014 2.0368 -0.3326>, rotation=<rot -0.0537 23.3309 90.124>}
  {shapeType=Box, extentX=2.75, extentY=21.75, extentZ=10.75, center=<-0.8272 -2.6249 1.0575>, rotation=<rot -0.4468 27.7184 2.3347>}
 ]
 dyn_rope_01_l [{shapeType=Capsule, length=2.4054, radius=1.6619, center=<2.1345 0 0.0001>, rotation=<rot 90 0 0>}]
 dyn_rope_01_r [{shapeType=Capsule, length=1.5575, radius=1.448, center=<1.722 0 -0.0001>, rotation=<rot 90 0 0>}]
 dyn_rope_02_l [{shapeType=Capsule, length=2.0206, radius=1.4001, center=<2.4456 0 0>, rotation=<rot 90 0 0>}]
 dyn_rope_02_r [{shapeType=Capsule, length=1.3389, radius=1.6412, center=<1.6338 0 -0.0002>, rotation=<rot 90 0 0>}]
 dyn_rope_03_l [{shapeType=Capsule, length=2.0121, radius=2.0207, center=<2.4692 0 0>, rotation=<rot 90 0 0>}]
 dyn_rope_03_r [{shapeType=Capsule, length=1.25, radius=1.5029, center=<1.5105 0 0.0001>, rotation=<rot 90 0 0>}]
 dyn_rope_base [{shapeType=Sphere, radius=1.375}]
 dyn_underpack_rop_01 [{shapeType=Capsule, length=1.76, radius=1.625, center=<0 0 -1.6133>}]
 dyn_underpack_rope_base [{shapeType=Sphere, radius=0.5}]
 dyn_underpack_rope_loop_01 [{shapeType=Capsule, length=0.4149, radius=1.4459, center=<1.4672 0 0>, rotation=<rot 90 0 0>}]
 dyn_underpack_rope_loop_02 [{shapeType=Capsule, length=0.3633, radius=1.6335, center=<1.5592 0 0>, rotation=<rot 90 0 0>}]
 dyn_underpack_rope_loop_base [{shapeType=Sphere, radius=0.5}]
 dyn_underpack_rope_loop_parent [{shapeType=Capsule, length=1.4036, radius=1.5387, center=<1.8307 0 0>, rotation=<rot 90 0 0>}]
 dyn_underpack_rope_parent [{shapeType=Capsule, length=1.7699, radius=2.3572, center=<2.4697 0 0>, rotation=<rot 0 0 90>}]
 dyn_zipper_base [{shapeType=Sphere, radius=0.5}]
 dyn_zipper_parent [{shapeType=Box, extentX=6.6837, extentY=3.951, extentZ=1, center=<3.2643 0 0.0001>}]
}
constraints [
 {bone1Name=dyn_bottle, bone2Name=dyn_bottle_carabiner_parent, swing1LimitDegrees=20, swing2LimitDegrees=20, twistLimitDegrees=3}
 {bone1Name=dyn_bottle_carabiner_parent, bone2Name=dyn_bottle_carabiner_base, swing1LimitDegrees=5, swing2LimitDegrees=10, twistLimitDegrees=3}
 {bone1Name=dyn_zipper_parent, bone2Name=dyn_zipper_base, swing1LimitDegrees=20, swing2LimitDegrees=10, twistLimitDegrees=3}
 {bone1Name=dyn_rope_01_l, bone2Name=dyn_rope_base, swing1LimitDegrees=10, swing2LimitDegrees=10, twistLimitDegrees=3}
 {bone1Name=dyn_rope_02_l, bone2Name=dyn_rope_01_l, swing1LimitDegrees=10, swing2LimitDegrees=10, twistLimitDegrees=3}
 {bone1Name=dyn_rope_03_l, bone2Name=dyn_rope_02_l, swing1LimitDegrees=15, swing2LimitDegrees=10, twistLimitDegrees=3}
 {bone1Name=dyn_rope_01_r, bone2Name=dyn_rope_base, swing1LimitDegrees=15, swing2LimitDegrees=10, twistLimitDegrees=3}
 {bone1Name=dyn_rope_02_r, bone2Name=dyn_rope_01_r, swing1LimitDegrees=15, swing2LimitDegrees=10, twistLimitDegrees=3}
 {bone1Name=dyn_rope_03_r, bone2Name=dyn_rope_02_r, swing1LimitDegrees=10, swing2LimitDegrees=10, twistLimitDegrees=3}
 {bone1Name=dyn_underpack_rope_parent, bone2Name=dyn_underpack_rope_base, swing1LimitDegrees=20, swing2LimitDegrees=20, twistLimitDegrees=3}
 {bone1Name=dyn_underpack_rop_01, bone2Name=dyn_underpack_rope_parent, swing1LimitDegrees=10, swing2LimitDegrees=10, twistLimitDegrees=3}
 {
  bone1Name=dyn_underpack_rope_loop_parent
  bone2Name=dyn_underpack_rope_loop_base
  swing1LimitDegrees=10
  swing2LimitDegrees=10
  twistLimitDegrees=3
 }
 {
  bone1Name=dyn_underpack_rope_loop_01
  bone2Name=dyn_underpack_rope_loop_parent
  swing1LimitDegrees=20
  swing2LimitDegrees=15
  twistLimitDegrees=3
 }
 {
  bone1Name=dyn_underpack_rope_loop_02
  bone2Name=dyn_underpack_rope_loop_01
  swing1LimitDegrees=15
  swing2LimitDegrees=15
  twistLimitDegrees=3
 }
 {bone1Name=dyn_pack_loop_parent, bone2Name=dyn_pack_loop_base, swing1LimitDegrees=10, swing2LimitDegrees=10, twistLimitDegrees=3}
]
